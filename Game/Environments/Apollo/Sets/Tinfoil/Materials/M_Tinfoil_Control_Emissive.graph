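# M_Tinfoil_Control_Emissive is a two-sided Default Lit material for the Apollo Tinfoil set, apparently for glowing control elements.
# A constant light-blue color is multiplied by a scalar to brighten it, presumably for emissive, alongside two standalone scalar constants.
# It uses no textures or parameters and is flagged for use with Nanite.
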Material /Game/Environments/Apollo/Sets/Tinfoil/Materials/M_Tinfoil_Control_Emissive
graph {
 properties {bUsedWithNanite=true, pixelDepthOffsetMode=PDOM_Legacy, twoSided=true, usedShadingModels=Default Lit}
 nodes [
  Constant() {r=1}
  Constant()
  Multiply(A=Constant3Vector() {constant=<rgba 0.2741 0.4786 0.849 0>}, B=Constant() {r=3})
 ]
}
