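# Material function for the Apollo Tinfoil set that draws one digit at an adjustable UV position and scale.
# Texture coordinates are offset and tiled per axis, then shifted by the rounded SingleDigit input and divided by 11 to pick a column of the T_Numbers2_M strip.
# Two one-dimensional SphereMasks multiply into a rectangular crop, so only that single cell shows in the masked RGB output.
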MaterialFunction /Game/Environments/Apollo/Sets/Tinfoil/Materials/MF_Tinfoil_Digit
graph {nodes [
  Constant2Vector() {r=5}
  n1 = FunctionInput(Preview=Constant() {r=5}) {bUsePreviewValueAsDefault=true, inputName=TilingScale_U, inputType=FunctionInput_Scalar}
  n2 = FunctionInput(Preview=Constant() {r=0.1}) {bUsePreviewValueAsDefault=true, inputName=PositionOffset_U, inputType=FunctionInput_Scalar}
  n3 = FunctionInput(Preview=Constant() {r=0.3}) {bUsePreviewValueAsDefault=true, inputName=PositionOffset_V, inputType=FunctionInput_Scalar}
  n4 = Add(A=TextureCoordinate(), B=Multiply(A=AppendVector(A=@n2, B=@n3)) {constB=-1})
  n5 = MaterialFunctionCall(Float2=@n4) {materialFunction=&BreakOutFloat2Components}
  n6 = FunctionInput(Preview=Constant() {r=2.5}) {bUsePreviewValueAsDefault=true, inputName=TilingScale_V, inputType=FunctionInput_Scalar}
  n7 = MaterialFunctionCall(Float2=AppendVector(A=Multiply(A=@n1, B=@n5.R), B=Multiply(A=@n5.G, B=@n6))) {materialFunction=&BreakOutFloat2Components}
  n8 = Divide() {constA=0.5, constB=11}
  n9 = SphereMask(A=Divide(A=@n7.R) {constB=11}, B=@n8, Radius=@n8)
  n10 = Constant() {r=0.5}
  n11 = SphereMask(A=@n7.G, B=@n10, Radius=@n10)
  n12 = Round(None=FunctionInput() {inputName=SingleDigit, inputType=FunctionInput_Scalar})
  n13 = AppendVector(A=Divide(A=Add(A=@n7.R, B=@n12) {constB=2}) {constB=11}, B=@n7.G)
  FunctionOutput(None=Multiply(
    A=Multiply(A=@n9, B=@n11)
    B=TextureSample(UVs=@n13) {samplerType=SAMPLERTYPE_Masks, texture=/Game/Athena/Playlists/Barrier/Textures/T_Numbers2_M}.RGB
   ))
  Multiply(A=@n4, B=@n1) {constB=5}
 ]}
