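# Unlit material for a Tinfoil countdown screen, likely the drill console, with countdown and ready looks plus a radial progress ring.
# InputSeconds is split into minute and second digits drawn by MF_Tinfoil_Digit, and a comparison against zero selects the output.
# Screens tint halves of one screen texture and use scanlines.
# A sphere-mask ring offers smooth or blocky angular fill, driven by looping time or custom input through static switches.
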Material /Game/Environments/Apollo/Sets/Tinfoil/Materials/M_Tinfoil_Countdown_Progress
graph {
 properties {
  bUsedWithNanite=true
  pixelDepthOffsetMode=PDOM_Legacy
  previewMesh=/Game/Environments/Apollo/Sets/Tinfoil/Meshes/SM_Tinfoil_DrillConsole_01
  shadingModel=MSM_Unlit
  usedShadingModels=Unlit
 }
 nodes [
  n1 = Time()
  n2 = Sine(None=Frac(None=Multiply(A=@n1) {constB=0.85}))
  Abs(None=@n2)
  n3 = NamedRerouteUsage()
  n4 = ScalarParameter() {defaultValue=60, parameterName=InputSeconds}
  If(A=@n4, B=Constant(), "A > B"=NamedRerouteUsage(), "A == B"=@n3, "A < B"=@n3)
  n5 = TextureCoordinate() {uTiling=2, vTiling=2}
  n6 = Constant() {r=1}
  n7 = MaterialFunctionCall(Vector or UVs=MaterialFunctionCall(UVs=Subtract(A=@n5, B=@n6), Rotation Center=Constant(), "Rotation Angle (0-1)"=Constant() {r=0.25}) {materialFunction=&CustomRotator}.Rotated Values) {materialFunction=&VectorToRadialValue}
  n8 = Subtract(
   A=SphereMask(A=@n7.Linear Distance, B=Constant(), Radius=ScalarParameter() {defaultValue=0.9, parameterName=OuterCircleRadius})
   B=SphereMask(A=@n7.Linear Distance, B=Constant(), Radius=ScalarParameter() {defaultValue=0.7, parameterName=InnerCircleRadius})
  )
  NamedRerouteDeclaration(None=@n8) {name=Full Circle, nodeColor=<rgba 0.8516 1 0 1>}
  n9 = TextureCoordinate()
  n10 = Constant() {r=0.5}
  n11 = MaterialFunctionCall(Vector or UVs=MaterialFunctionCall(UVs=Subtract(A=@n9, B=@n10), Rotation Center=Constant(), "Rotation Angle (0-1)"=Constant() {r=0.25}) {materialFunction=&CustomRotator}.Rotated Values) {materialFunction=&VectorToRadialValue}
  n12 = Constant()
  n13 = Constant() {r=1}
  n14 = Time()
  n15 = StaticSwitchParameter(True=ScalarParameter() {parameterName="CustomProgressInput (0-1)"}, False=Frac(None=Multiply(A=@n14) {constB=0.2})) {parameterName=UseCustomInput}
  NamedRerouteDeclaration(None=If(A=@n11.Vector Converted to Angle, B=@n15, "A > B"=@n12, "A == B"=@n13, "A < B"=@n13)) {name=SmoothRadialProgress, nodeColor=<rgba 0.1953 0 1 1>}
  n16 = Constant() {r=0.07}
  n17 = Constant() {r=60}
  n18 = Floor(None=Divide(A=@n4, B=@n17))
  n19 = Constant() {r=10}
  n20 = NamedRerouteDeclaration(None=Floor(None=Divide(A=@n18, B=@n19))) {name=Minute_Digit1, nodeColor=<rgba 1 0 0.1875 1>}
  n21 = Constant() {r=0.25}
  n22 = NamedRerouteDeclaration(None=Fmod(A=@n18, B=@n19)) {name=Minute_Digit2, nodeColor=<rgba 1 0.2812 0 1>}
  n23 = Fmod(A=@n4, B=@n17)
  n24 = Constant() {r=10}
  n25 = Floor(None=Divide(A=@n23, B=@n24))
  n26 = Add(
   A=Add(
    A=MaterialFunctionCall!(PositionOffset_U=@n16, PositionOffset_V, TilingScale_U, TilingScale_V, SingleDigit=@n20) {materialFunction=&/MF_Tinfoil_Digit}.Result
    B=MaterialFunctionCall(PositionOffset_U=@n21, SingleDigit=@n22) {materialFunction=&MF_Tinfoil_Digit}.Result
   )
   B=MaterialFunctionCall(PositionOffset_U=Constant() {r=0.55}, SingleDigit=NamedRerouteDeclaration(None=@n25) {name=Second_Digit1, nodeColor=<rgba 0 0.9766 1 1>}) {materialFunction=&MF_Tinfoil_Digit}.Result
  )
  n27 = MaterialFunctionCall(
   PositionOffset_U=Constant() {r=0.74}
   SingleDigit=NamedRerouteDeclaration(None=Fmod(A=@n23, B=@n24)) {name=Second_Digit2, nodeColor=<rgba 0 1 0.9531 1>}
  ) {materialFunction=&MF_Tinfoil_Digit}
  n28 = Constant() {r=0.43}
  n29 = Constant() {r=10}
  NamedRerouteDeclaration(None=Add(
    A=Add(A=@n26, B=@n27.Result)
    B=MaterialFunctionCall(PositionOffset_U=@n28, SingleDigit=@n29) {materialFunction=&MF_Tinfoil_Digit}.Result
   )) {name=Timer, nodeColor=<rgba 0.5703 1 0 1>}
  n30 = Constant() {r=0.8}
  n31 = Constant() {r=0.05}
  NamedRerouteDeclaration(None=Multiply(A=MaterialFunctionCall!(ScanLine_Tiling_1=@n30, Scanline_PannerSpeed_1=@n31, ScanLine_Tiling, Scanline_PannerSpeed=@n31, UVs) {materialFunction=/Game/Weapons/WeaponSkins/MaterialFunction/ScanLines}.Result) {constB=0.1}) {name=Scanned Lines, nodeColor=<rgba 0.8281 0 1 1>}
  n32 = MaterialFunctionCall(Float2=TextureCoordinate()) {materialFunction=&BreakOutFloat2Components}
  n33 = Multiply(A=@n32.G) {constB=0.5}
  n34 = TextureSample(UVs=AppendVector(A=@n32.R, B=@n33)) {texture=/Game/Environments/Apollo/Sets/Complex/Textures/Screens/T_Complex_Screen_B_TA_M}
  n35 = LinearInterpolate(A=Power(Base=@n34.R) {constExponent=1.3}, B=Power(Base=@n34.R) {constExponent=1.5}, Alpha=@n2)
  n36 = Multiply(A=NamedRerouteUsage()) {constB=0.2}
  NamedRerouteDeclaration(None=Multiply(A=Add(A=@n35, B=@n36), B=VectorParameter() {parameterName=ScreenColor_Ready, defaultValue=<rgba 0.0365 1 0.0826 1>}.RGB)) {name=Ready Screen, nodeColor=<rgba 0.4766 1 0 1>}
  n37 = AppendVector(A=@n32.R, B=Add(A=@n33) {constB=0.5})
  n38 = NamedRerouteUsage()
  n39 = ScalarParameter() {defaultValue=2, parameterName=Timer_EmissiveMult}
  n40 = Add(
   A=Add(
    A=TextureSample(UVs=@n37) {texture=&T_Complex_Screen_B_TA_M}.R
    B=Multiply(A=@n38, B=@n39)
   )
   B=NamedRerouteUsage()
  )
  n41 = VectorParameter() {parameterName=ScreenColor_CD, defaultValue=<rgba 0.1667 0 0.0016 1>}
  NamedRerouteDeclaration(None=Multiply(A=Multiply(A=@n40, B=@n41.RGB)) {constB=2}) {name=CD Screen, nodeColor=<rgba 1 0 0.8906 1>}
  n42 = MaterialFunctionCall(Float2=@n7.Vector Converted to Angle) {materialFunction=&BreakOutFloat2Components}
  n43 = Frac(None=Multiply(A=@n42.R, B=ScalarParameter() {defaultValue=10, parameterName=BlockAmount}) {constB=10})
  n44 = ScalarParameter() {defaultValue=0.1, parameterName=BlockWidth}
  n45 = Constant() {r=1}
  n46 = Constant()
  NamedRerouteDeclaration(None=Multiply(A=@n8, B=If(A=@n43, B=@n44, "A > B"=@n45, "A == B"=@n46, "A < B"=@n46))) {name=Block Circle, nodeColor=<rgba 0 1 0.9531 1>}
  NamedRerouteDeclaration(None=@n42.R) {name=RadialGradient0-1, nodeColor=<rgba 0 0.4375 1 1>}
  n47 = ScalarParameter() {defaultValue=10, parameterName=BlockAmount}
  n48 = Divide(B=@n47)
  n49 = Floor(None=Multiply(A=Add(A=@n15, B=@n48) {constB=0.1}, B=@n47))
  NamedRerouteDeclaration(None=If(A=@n11.Vector Converted to Angle, B=Divide(A=@n49, B=@n47) {constB=5}, "A > B"=@n12, "A == B"=@n13, "A < B"=@n13)) {name=BlockRadialProgress, nodeColor=<rgba 1 0.2344 0 1>}
  NamedRerouteDeclaration(None=@n15) {name=Progression, nodeColor=<rgba 0.0312 1 0 1>}
  n50 = VectorParameter() {parameterName=Color1, defaultValue=<rgba 0.6562 0.0401 0 1>}
  n51 = VectorParameter() {parameterName=Color2, defaultValue=<rgba 0.0068 0.6094 0 1>}
  n52 = NamedRerouteUsage()
  n53 = NamedRerouteUsage()
  n54 = NamedRerouteUsage()
  n55 = VectorParameter() {parameterName=Color1, defaultValue=<rgba 0.6562 0.0401 0 1>}
  n56 = VectorParameter() {parameterName=Color2, defaultValue=<rgba 0.0068 0.6094 0 1>}
  n57 = NamedRerouteUsage()
  n58 = NamedRerouteUsage()
  n59 = NamedRerouteUsage()
  StaticSwitchParameter(
   True=Multiply(A=LinearInterpolate(A=@n50.RGB, B=@n51.RGB, Alpha=@n52), B=Multiply(A=@n53, B=@n54))
   False=Multiply(A=LinearInterpolate(A=@n55.RGB, B=@n56.RGB, Alpha=@n57), B=Multiply(A=@n58, B=@n59))
  ) {parameterName=Use SmoothProgression}
 ]
}
