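# Material for the Tinfoil set's electrical box fan, which spins the mesh continuously in the vertex shader.
# World position offset rotates vertices around the object pivot on a parameterized local axis, with FixRotateAboutAxisNormals correcting the normal-mapped shading.
# Spin rate is time scaled by a clamped value hashed from object X and Y position, so placed instances vary.
# The mask texture's green channel feeds metallic, specular and roughness alike, through Fortnite_BaseFunction.
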Material /Game/Environments/Apollo/Sets/Tinfoil/Materials/M_Tinfoil_ElectricalBox_01_FanSpin
graph {
 properties {
  bUseMaterialAttributes=true
  bUsedWithNanite=true
  physMaterial=/Game/FortressPhysicalMaterials/Metal
  pixelDepthOffsetMode=PDOM_Legacy
  refractionMethod=RM_IndexOfRefraction
  usedShadingModels=Default Lit
 }
 nodes [
  LinearInterpolate()
  n1 = TextureSampleParameter2D() {
   parameterName=SpecularMasks
   samplerType=SAMPLERTYPE_Masks
   texture=/Game/Environments/Apollo/Sets/Tinfoil/Textures/T_Apollo_Tinfoil_ElectricalBox_01_Fan_S
  }
  n2 = Normalize(VectorInput=Transform(None=VectorParameter() {parameterName=Vector Rotation Direction, defaultValue=<rgba 0 1 0 1>}.RGB) {transformSourceType=TRANSFORMSOURCE_Local})
  n3 = ObjectPositionWS()
  n4 = ScalarParameter() {defaultValue=70, parameterName=RandomScale}
  n5 = Frac(None=Divide(A=ComponentMask(None=@n3) {r=true}, B=@n4))
  n6 = Sine(None=Divide(A=ComponentMask(None=@n3) {g=true}, B=@n4))
  n7 = Multiply(A=Clamp(None=Add(A=@n5, B=@n6)) {minDefault=0.2}, B=Time())
  n8 = TextureSampleParameter2D() {
   parameterName=Normals
   samplerType=SAMPLERTYPE_Normal
   texture=/Game/Environments/Apollo/Sets/Tinfoil/Textures/T_Apollo_Tinfoil_ElectricalBox_01_Fan_N
  }
  n9 = Transform(None=@n8.RGB)
  n10 = MaterialFunctionCall() {materialFunction=&ObjectPivotPoint}
  n11 = WorldPosition()
  MaterialFunctionCall!(
   Material attributes=MakeMaterialAttributes!(
    BaseColor=TextureSampleParameter2D() {parameterName=Diffuse, texture=/Game/Environments/Apollo/Sets/Tinfoil/Textures/T_Apollo_Tinfoil_ElectricalBox_01_Fan_D}.RGB
    Metallic=@n1.G
    Specular=@n1.G
    Roughness=@n1.G
    Anisotropy
    EmissiveColor
    Opacity
    OpacityMask
    Normal=MaterialFunctionCall(Rotation Axis=@n2, Rotation Angle=@n7, World Space Vertex Normals=@n9) {materialFunction=&FixRotateAboutAxisNormals}.New Tangent Space Vertex Normal
    Tangent
    WorldPositionOffset=RotateAboutAxis(NormalizedRotationAxis=@n2, RotationAngle=@n7, PivotPoint=@n10.Object Pivot Location, Position=@n11.XYZ)
    SubsurfaceColor
    ClearCoat
    ClearCoatRoughness
    AmbientOcclusion
    Refraction
    CustomizedUVs_0
    CustomizedUVs_1
    CustomizedUVs_2
    CustomizedUVs_3
    CustomizedUVs_4
    CustomizedUVs_5
    CustomizedUVs_6
    CustomizedUVs_7
    PixelDepthOffset
    ShadingModel
    Displacement
   )
   IsGameplayAsset=StaticBool() {value=true}
  ) {materialFunction=/Game/Packages/Fortress_SharedMaterials/Base_Material_Functions/Fortnite_BaseFunction}
 ]
}
