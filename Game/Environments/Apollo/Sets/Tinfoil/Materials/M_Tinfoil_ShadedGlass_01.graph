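# Translucent, per-pixel lit shaded glass material for the Apollo Tinfoil environment set.
# Scalar and vector parameters expose specular, opacity and a default dark grey tint.
# The green channel of a glass grime texture, tiled twice, blends between two low constants to add surface variation.
# A cloud noise mask is sampled on the same UVs, though the graph shows no use of it.
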Material /Game/Environments/Apollo/Sets/Tinfoil/Materials/M_Tinfoil_ShadedGlass_01
graph {
 properties {
  blendMode=BLEND_Translucent
  pixelDepthOffsetMode=PDOM_Legacy
  translucencyLightingMode=TLM_SurfacePerPixelLighting
  usedShadingModels=Default Lit
 }
 nodes [
  n1 = TextureCoordinate() {uTiling=2, vTiling=2}
  LinearInterpolate(Alpha=TextureSample(UVs=@n1) {samplerType=SAMPLERTYPE_LinearColor, texture=/Game/Environments/Apollo/Sets/Tinfoil/Textures/T_Tinfoil_GlassGrime_M}.G) {constA=0.05, constB=0.4}
  ScalarParameter() {defaultValue=0.9, parameterName=GlassSpec}
  ScalarParameter() {defaultValue=0.66, parameterName=GlassOpacity}
  TextureSample(UVs=@n1) {samplerType=SAMPLERTYPE_Masks, texture=/Game/Global/Textures/Default/Noise/T_Noise_Cloud_Mask}
  VectorParameter() {parameterName=Glass Tint, defaultValue=<rgba 0.1094 0.1094 0.1094 1>}
 ]
}
